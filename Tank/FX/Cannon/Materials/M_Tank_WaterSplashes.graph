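# Translucent unlit Niagara sprite material for water splashes in the Tank cannon effects.
# A desaturated, tinted sky cubemap reflection (constant on mobile) gains a sun/moon reflection-vector highlight, then particle color and emissive adjustments, faking lit water without lighting.
# A random sub-UV frame of the splash texture is multiplied by near-camera fade and particle alpha to form an opacity mask.
# A Fresnel-times-cubemap branch feeds no other node.
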Material /Tank/FX/Cannon/Materials/M_Tank_WaterSplashes
graph {
 properties {bUsedWithNiagaraSprites=true, blendMode=BLEND_Translucent, pixelDepthOffsetMode=PDOM_Legacy, shadingModel=MSM_Unlit, usedShadingModels=Unlit}
 nodes [
  n1 = TextureSampleParameterCube!(
   UVs=ReflectionVectorWS() {bNormalizeCustomWorldNormal=true}
   Level=ScalarParameter() {defaultValue=2, parameterName=Bias}
   Apply View MipBias
  ) {
   mipValueMode=TMVM_MipLevel
   parameterName=MaterialExpressionTextureSampleParameterCube_0
   texture=/Game/Packages/Fortress_Sky/TexturesHDR/T_hdrsky_test03_low
  }
  n2 = Multiply(A=@n1.RGB, B=VectorParameter() {parameterName=Cube Map Multiplier, defaultValue=<rgba 15 15 15 0>}.RGB) {constB=1.5}
  n3 = Clamp(None=@n2)
  n4 = ComponentMask(None=MaterialFunctionCall() {materialFunction=/Engine/Functions/GetPostProcessFunctions/GetAmbientCubemapTint}.AmbientCubemapTint) {b=true, g=true, r=true}
  n5 = ShadingPathSwitch(
   Default=Desaturation(None=Multiply(A=@n3, B=@n4), Fraction=ScalarParameter() {defaultValue=1, parameterName=Desaturation Factor})
   Mobile=Constant() {r=1}
  ) {desc=Migrated from FeatureLevelSwitch}
  n6 = CollectionParameter() {
   collection=&FortniteMaterialParameters
   parameterName=SunAndMoonModelDirectionalVector
  }
  n7 = ComponentMask(None=@n6) {b=true, g=true, r=true}
  n8 = ReflectionVectorWS()
  n9 = Power(Base=Clamp(None=DotProduct(A=@n7, B=@n8)), Exp=ScalarParameter() {defaultValue=5, parameterName=Spec Highlight Power}) {constExponent=20}
  n10 = ScalarParameter() {defaultValue=25, parameterName=Spec Intense}
  n11 = Multiply(A=@n5, B=Clamp(None=Multiply(A=@n9, B=@n10)) {minDefault=0.7})
  MaterialFunctionCall(Emissive=Multiply(A=Add(A=@n5, B=@n11), B=ParticleColor().RGB)) {materialFunction=&MF_FX_EmissiveAdjustments}
  n12 = MaterialFunctionCall!(
   UseTexCoord 1
   Rows=ScalarParameter() {defaultValue=2, parameterName="Rows (S)"}
   Columns=ScalarParameter() {defaultValue=2, parameterName="Columns (S)"}
   Particle Random Value=ParticleRandom()
  ) {materialFunction=/Game/Effects/Fort_Effects/Materials/_Functions/MF_RandomSubUV_GPU_Sprites}
  n13 = MaterialFunctionCall(Distance=ScalarParameter() {defaultValue=125, parameterName="Distance (S)"}) {materialFunction=&Near_Camera_Fade}
  Multiply(
   A=ParticleColor().A
   B=Multiply(
    A=TextureSample(UVs=@n12.Result) {samplerType=SAMPLERTYPE_LinearColor, texture=/Game/Effects/Fort_Effects/Textures/Water/T_waterSplash}.R
    B=Clamp(None=@n13.Result)
   )
  )
  Multiply(A=Fresnel() {exponent=2}, B=@n2)
 ]
}
